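# Animation Blueprint for the MurkyVole backpack, animating its jet thrusters and wings, missile-swarm rocket pods and auto-targeting turret.
# Three state machines are layered per bone onto the default pack pose, then pass through DefaultSlot.
# Turret and rocket states handle equip, activation and unequip, aiming bones with world-space rotations; turret attacks nest a fire machine and rockets stow during AshtonShield blocking.
# Driving variables are read but not declared here, and the EventGraph is empty.
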AnimBlueprint /MurkyVoleGameplay/Assets/Backpack/Meshes/M_MED_MurkyVole_Pack_AnimBP
target_skeleton=&/M_MED_MurkyVole_Pack_Skeleton
graphs [
 AnimGraph {nodes [
   n1 = "Transform (Modify) Bone - Bone: turret_servo_base"(ComponentPose=Local To Component(), Alpha=1, Translation="0,0,0", Rotation="0,0,30", Scale="1,1,1") {
    rotationMode=BMM_Additive
    rotationSpace=BCS_ParentBoneSpace
    alphaBoolBlend {blendOption=Linear}
    boneToModify {boneName=turret_servo_base}
    rotation {roll=30}
   }
   Component To Local(ComponentPose=@n1)
   AnimGraphNode_StateMachine_0 = JetPacks_SM() {is=title AnimGraphNode_StateMachine}
   n2 = Layered blend per bone(
    BasePose=Sequence Player 'MurkyVole_Pack_Default'() {bLoopAnimation=false, sequence=/MurkyVoleGameplay/Assets/Backpack/Animations/MurkyVole_Pack_Default}
    BlendPoses_0=Blend Poses by bool(
     BlendPose_0=@AnimGraphNode_StateMachine_0
     BlendPose_1=Sequence Player 'MurkyVole_Pack_Default'() {bLoopAnimation=false, sequence=&MurkyVole_Pack_Default}
     BlendTime_0=0.2
     BlendTime_1=0.2
    ) {blendType=Linear}
   ) {layerSetup [{branchFilters [{boneName=jet_thruster_mstr_l}, {boneName=jet_thruster_mstr_r}, {boneName=jet_wing_mstr_l}, {boneName=jet_wing_mstr_r}]}]}
   AnimGraphNode_StateMachine_2 = RocketPod SM() {is=title AnimGraphNode_StateMachine}
   n3 = Layered blend per bone(BasePose=@n2, BlendPoses_0=@AnimGraphNode_StateMachine_2) {layerSetup [{branchFilters [{boneName=rocket_servo_01_l}, {boneName=rocket_servo_01_r}]}]}
   AnimGraphNode_StateMachine_1 = Turret SM() {is=title AnimGraphNode_StateMachine}
   Output Pose(Result=Slot 'DefaultSlot'(Source=Layered blend per bone(BasePose=@n3, BlendPoses_0=@AnimGraphNode_StateMachine_1) {layerSetup [{branchFilters [{boneName=turret_servo_base}]}]}))
  ]}
 JetPacks_SM {
  entry=Equip
  path=AnimGraph.AnimGraphNode_StateMachine_0.JetPacks_SM
  ids=id
  nodes [
   AnimStateNode_0 = Equip() {is=title AnimStateNode}
   AnimStateNode_2 = Deployed() {is=title AnimStateNode}
   AnimStateNode_3 = Equipped Pose() {is=title AnimStateNode}
   AnimStateNode_4 = Retract() {is=title AnimStateNode}
   AnimStateNode_6 = Retracted Pose() {is=title AnimStateNode}
   AnimStateTransitionNode_1 = AnimStateTransitionNode() {bAutomaticRuleBasedOnSequencePlayerInState=true, crossfadeDuration=0}
   AnimStateTransitionNode_5 = AnimStateTransitionNode() {blendMode=Linear, crossfadeDuration=0}
   AnimStateTransitionNode_6 = AnimStateTransitionNode()
   AnimStateTransitionNode_8 = AnimStateTransitionNode() {bAutomaticRuleBasedOnSequencePlayerInState=true, blendMode=Linear, crossfadeDuration=0}
  ]
  links [
   {
    dst=Deployed
    rule=AnimGraph.AnimGraphNode_StateMachine_0.JetPacks_SM.AnimStateTransitionNode_1.Transition
    src=Equip
    via=AnimStateTransitionNode_1
   }
   {
    dst=Retract
    rule=AnimGraph.AnimGraphNode_StateMachine_0.JetPacks_SM.AnimStateTransitionNode_5.Transition
    src=Equipped Pose
    via=AnimStateTransitionNode_5
   }
   {
    dst=Retracted Pose
    rule=AnimGraph.AnimGraphNode_StateMachine_0.JetPacks_SM.AnimStateTransitionNode_8.Transition
    src=Retract
    via=AnimStateTransitionNode_8
   }
   {
    dst=Equipped Pose
    rule=AnimGraph.AnimGraphNode_StateMachine_0.JetPacks_SM.AnimStateTransitionNode_6.Transition
    src=Retracted Pose
    via=AnimStateTransitionNode_6
   }
  ]
 }
 Equip {
  path=AnimGraph.AnimGraphNode_StateMachine_0.JetPacks_SM.AnimStateNode_0.Equip
  nodes [Output Animation Pose(Result=Sequence Player 'MurkyVole_Pack_Jets_Equip_Weapon'() {bLoopAnimation=false, sequence=/MurkyVoleGameplay/Assets/Backpack/Animations/Jets/MurkyVole_Pack_Jets_Equip_Weapon})]
 }
 Deployed {
  path=AnimGraph.AnimGraphNode_StateMachine_0.JetPacks_SM.AnimStateNode_2.Deployed
  nodes [
   Sequence Player 'MurkyVole_Pack_Jets_Deployed'() {sequence=/MurkyVoleGameplay/Assets/Backpack/Animations/MurkyVole_Pack_Jets_Deployed}
   n1 = Blend Poses by bool(
    BlendPose_0=Blendspace Player 'MurkyVole_Pack_Jets_Boost_BS'(X=0, Y=0) {blendSpace=/MurkyVoleGameplay/Assets/Backpack/Animations/Jets/MurkyVole_Pack_Jets_Boost_BS}
    BlendPose_1=Blend Poses by bool(
     BlendPose_0=Blendspace Player 'MurkyVole_Pack_Jets_Leans_BS'(X=0, Y=0) {blendSpace=/MurkyVoleGameplay/Assets/Backpack/Animations/Jets/MurkyVole_Pack_Jets_Leans_BS}
     BlendPose_1=Sequence Player 'MurkyVole_Pack_Jets_Base_Pose_Weapon'() {bLoopAnimation=false, sequence=/MurkyVoleGameplay/Assets/Backpack/Animations/Jets/MurkyVole_Pack_Jets_Base_Pose_Weapon}
    ) {blendType=Linear}
    BlendTime_1=0.3
   ) {blendType=Linear}
   n2 = "Sequence Player (additive) 'MurkyVole_Pack_Jets_Jitter_Additive_Weapon'"() {sequence=/MurkyVoleGameplay/Assets/Backpack/Animations/Jets/MurkyVole_Pack_Jets_Jitter_Additive_Weapon}
   Output Animation Pose(Result=Apply Additive!(
     Base=Apply Additive(Base=@n1, Additive=@n2, bAlphaBoolEnabled=True) {alphaInputType=Bool, alphaBoolBlend {blendOption=Linear}}
     Additive="Sequence Player (additive) 'MurkyVole_Pack_Jets_JitterBoost_Weapon'"() {sequence=/MurkyVoleGameplay/Assets/Backpack/Animations/Jets/MurkyVole_Pack_Jets_JitterBoost_Weapon}
     bAlphaBoolEnabled:bool=True
    ) -> (Pose) {alphaInputType=Bool, alphaBoolBlend {blendOption=Linear}})
  ]
 }
 Equipped Pose {
  path=AnimGraph.AnimGraphNode_StateMachine_0.JetPacks_SM.AnimStateNode_3.Equipped Pose
  nodes [Output Animation Pose(Result=Sequence Player 'MurkyVole_Pack_Jets_Base_Pose_Weapon'() {bLoopAnimation=false, sequence=&MurkyVole_Pack_Jets_Base_Pose_Weapon})]
 }
 Retract {
  path=AnimGraph.AnimGraphNode_StateMachine_0.JetPacks_SM.AnimStateNode_4.Retract
  nodes [Output Animation Pose(Result=Sequence Player 'MurkyVole_Pack_Jets_Retract_Weapon'() {bLoopAnimation=false, sequence=/MurkyVoleGameplay/Assets/Backpack/Animations/Jets/MurkyVole_Pack_Jets_Retract_Weapon})]
 }
 Retracted Pose {
  path=AnimGraph.AnimGraphNode_StateMachine_0.JetPacks_SM.AnimStateNode_6.Retracted Pose
  nodes [
   Sequence Player 'MurkyVole_Pack_Default'() {sequence=&MurkyVole_Pack_Default}
   Output Animation Pose(Result=Sequence Player 'MurkyVole_Pack_Jets_Jog_Pose_Weapon'() {bLoopAnimation=false, sequence=/MurkyVoleGameplay/Assets/Backpack/Animations/Jets/MurkyVole_Pack_Jets_Jog_Pose_Weapon})
  ]
 }
 Transition {path=AnimGraph.AnimGraphNode_StateMachine_0.JetPacks_SM.AnimStateTransitionNode_1.Transition, nodes [Result()]}
 Transition {
  path=AnimGraph.AnimGraphNode_StateMachine_0.JetPacks_SM.AnimStateTransitionNode_5.Transition
  nodes [Result(bCanEnterTransition=Get bIsPawnSpeed2DAboveThreshold() -> (bIsPawnSpeed2DAboveThreshold:bool))]
 }
 Transition {
  path=AnimGraph.AnimGraphNode_StateMachine_0.JetPacks_SM.AnimStateTransitionNode_6.Transition
  nodes [Result(bCanEnterTransition=NOT Boolean(A=Get bIsPawnSpeed2DAboveThreshold()))]
 }
 Transition {path=AnimGraph.AnimGraphNode_StateMachine_0.JetPacks_SM.AnimStateTransitionNode_8.Transition, nodes [Result()]}
 Turret SM {
  entry=Undeployed
  path=AnimGraph.AnimGraphNode_StateMachine_1.Turret SM
  ids=id
  nodes [
   To Unequip() {is=title AnimStateAliasNode, also {aliases [Attacking, AutoTarget Off, Scanning]}}
   AnimStateNode_0 = Undeployed() {is=title AnimStateNode}
   AnimStateNode_1 = Scanning() {is=title AnimStateNode}
   AnimStateNode_2 = AutoTarget Off() {is=title AnimStateNode}
   AnimStateNode_3 = Attacking() {is=title AnimStateNode}
   AnimStateNode_4 = Equip()
   AnimStateNode_5 = UnEquip() {is=title AnimStateNode}
   AnimStateTransitionNode_0 = AnimStateTransitionNode()
   AnimStateTransitionNode_1 = AnimStateTransitionNode()
   AnimStateTransitionNode_10 = AnimStateTransitionNode()
   AnimStateTransitionNode_11 = AnimStateTransitionNode()
   AnimStateTransitionNode_13 = AnimStateTransitionNode() {automaticRuleTriggerTime=0, bAutomaticRuleBasedOnSequencePlayerInState=true}
   AnimStateTransitionNode_14 = AnimStateTransitionNode()
   AnimStateTransitionNode_2 = AnimStateTransitionNode()
   AnimStateTransitionNode_4 = AnimStateTransitionNode() {bAutomaticRuleBasedOnSequencePlayerInState=true, crossfadeDuration=0}
   AnimStateTransitionNode_6 = AnimStateTransitionNode()
   AnimStateTransitionNode_7 = AnimStateTransitionNode() {crossfadeDuration=0.5}
  ]
  links [
   {
    dst=AutoTarget Off
    rule=AnimGraph.AnimGraphNode_StateMachine_1.Turret SM.AnimStateTransitionNode_11.Transition
    src=Attacking
    via=AnimStateTransitionNode_11
   }
   {
    dst=Scanning
    rule=AnimGraph.AnimGraphNode_StateMachine_1.Turret SM.AnimStateTransitionNode_7.Transition
    src=Attacking
    via=AnimStateTransitionNode_7
   }
   {
    dst=Attacking
    rule=AnimGraph.AnimGraphNode_StateMachine_1.Turret SM.AnimStateTransitionNode_1.Transition
    src=AutoTarget Off
    via=AnimStateTransitionNode_1
   }
   {
    dst=Scanning
    rule=AnimGraph.AnimGraphNode_StateMachine_1.Turret SM.AnimStateTransitionNode_10.Transition
    src=AutoTarget Off
    via=AnimStateTransitionNode_10
   }
   {
    dst=Scanning
    rule=AnimGraph.AnimGraphNode_StateMachine_1.Turret SM.AnimStateTransitionNode_13.Transition
    src=Equip
    via=AnimStateTransitionNode_13
   }
   {
    dst=Attacking
    rule=AnimGraph.AnimGraphNode_StateMachine_1.Turret SM.AnimStateTransitionNode_6.Transition
    src=Scanning
    via=AnimStateTransitionNode_6
   }
   {
    dst=AutoTarget Off
    rule=AnimGraph.AnimGraphNode_StateMachine_1.Turret SM.AnimStateTransitionNode_0.Transition
    src=Scanning
    via=AnimStateTransitionNode_0
   }
   {
    dst=UnEquip
    rule=AnimGraph.AnimGraphNode_StateMachine_1.Turret SM.AnimStateTransitionNode_2.Transition
    src=To Unequip
    via=AnimStateTransitionNode_2
   }
   {
    dst=Undeployed
    rule=AnimGraph.AnimGraphNode_StateMachine_1.Turret SM.AnimStateTransitionNode_4.Transition
    src=UnEquip
    via=AnimStateTransitionNode_4
   }
   {
    dst=Equip
    rule=AnimGraph.AnimGraphNode_StateMachine_1.Turret SM.AnimStateTransitionNode_14.Transition
    src=Undeployed
    via=AnimStateTransitionNode_14
   }
  ]
 }
 Undeployed {
  path=AnimGraph.AnimGraphNode_StateMachine_1.Turret SM.AnimStateNode_0.Undeployed
  nodes [Output Animation Pose(Result=Sequence Player 'MurkyVole_Pack_Default'() {bLoopAnimation=false, sequence=&MurkyVole_Pack_Default})]
 }
 Scanning {
  path=AnimGraph.AnimGraphNode_StateMachine_1.Turret SM.AnimStateNode_1.Scanning
  nodes [
   Blend Poses by bool(
    BlendPose_0=Sequence Player 'MurkyVole_Pack_Sprinting'() {bLoopAnimation=false, sequence=/MurkyVoleGameplay/Assets/Backpack/Animations/MurkyVole_Pack_Sprinting}
    BlendPose_1=Blend Poses by bool(
     BlendPose_0=Sequence Player 'MurkyVole_Pack_Crouch_Moving_Rockets_Inactive'() {bLoopAnimation=false, sequence=/MurkyVoleGameplay/Assets/Backpack/Animations/MurkyVole_Pack_Crouch_Moving_Rockets_Inactive}
     BlendPose_1=Sequence Player 'MurkyVole_Pack_Turret_Deployed'() {bLoopAnimation=false, sequence=/MurkyVoleGameplay/Assets/Backpack/Animations/MurkyVole_Pack_Turret_Deployed}
     BlendTime_0=0.25
     BlendTime_1=0.4
    ) {blendType=Linear}
    BlendTime_0=0.2
    BlendTime_1=0.2
   ) {blendType=Linear}
   n1 = Blendspace Player 'MurkyVole_Turret_Pitch_Offset_BS'!(X=0) -> (Pose) {blendSpace=/MurkyVoleGameplay/Assets/Backpack/Animations/Turret/MurkyVole_Turret_Pitch_Offset_BS}
   n2 = Sequence Player 'MurkyVole_Pack_Turret_Deployed'() {bLoopAnimation=false, sequence=&MurkyVole_Pack_Turret_Deployed}
   n3 = "Transform (Modify) Bone - Bone: turret_rail"(
    ComponentPose=Local To Component(LocalPose=Apply Additive!(Base=@n2, Additive=@n1, Alpha:float) -> (Pose) {alphaBoolBlend {blendOption=Linear}})
    Alpha=1
    Translation="0,0,0"
    Rotation=Get TurretAimRotation() -> (TurretAimRotation:Rotator)
    Scale="1,1,1"
   ) {
    rotationMode=BMM_Replace
    rotationSpace=BCS_WorldSpace
    scaleSpace=BCS_WorldSpace
    translationSpace=BCS_WorldSpace
    alphaBoolBlend {blendOption=Linear}
    boneToModify {boneName=turret_rail}
   }
   Output Animation Pose(Result=Layered blend per bone(
     BasePose=Component To Local(ComponentPose=@n3)
     BlendPoses_0=Blendspace Player 'MurkyVole_Turret_Rail_BS'!(X=0) -> (Pose) {blendSpace=/MurkyVoleGameplay/Assets/Backpack/Animations/Turret/MurkyVole_Turret_Rail_BS}
    ) {layerSetup [{branchFilters [{boneName=turret_gun}, {blendDepth=-1, boneName=turret_barrel}]}]})
  ]
 }
 AutoTarget Off {
  path=AnimGraph.AnimGraphNode_StateMachine_1.Turret SM.AnimStateNode_2.AutoTarget Off
  nodes [
   n1 = Blend Poses by bool(
    BlendPose_0=Sequence Player 'MurkyVole_Pack_Crouch_Moving_Rockets_Inactive'() {bLoopAnimation=false, sequence=&MurkyVole_Pack_Crouch_Moving_Rockets_Inactive}
    BlendPose_1=Sequence Player 'MurkyVole_Pack_Turret_Deployed'() {bLoopAnimation=false, sequence=&MurkyVole_Pack_Turret_Deployed}
    BlendTime_0=0.25
    BlendTime_1=0.25
   ) {blendType=Linear}
   n2 = Sequence Player 'MurkyVole_Pack_Sprinting'() {bLoopAnimation=false, sequence=&MurkyVole_Pack_Sprinting}
   Output Animation Pose(Result=Layered blend per bone(
     BasePose=Blend Poses by bool(BlendPose_0=@n2, BlendPose_1=@n1, BlendTime_0=0.2, BlendTime_1=0.2) {blendType=Linear}
     BlendPoses_0=Sequence Player 'MurkyVole_Pack_Turret_AutoDisabled'() {bLoopAnimation=false, sequence=/MurkyVoleGameplay/Assets/Backpack/Animations/Turret/MurkyVole_Pack_Turret_AutoDisabled}
    ) {layerSetup [{branchFilters [{boneName=turret_servo}]}]})
  ]
 }
 Attacking {
  path=AnimGraph.AnimGraphNode_StateMachine_1.Turret SM.AnimStateNode_3.Attacking
  nodes [
   Blend Poses by bool(
    BlendPose_0=Sequence Player 'MurkyVole_Pack_Sprinting'() {bLoopAnimation=false, sequence=&MurkyVole_Pack_Sprinting}
    BlendPose_1=Sequence Player 'MurkyVole_Pack_Turret_Deployed'() {bLoopAnimation=false, sequence=&MurkyVole_Pack_Turret_Deployed}
   ) {blendType=Linear}
   n1 = Blendspace Player 'MurkyVole_Turret_Pitch_Offset_BS'(X=0) {blendSpace=&MurkyVole_Turret_Pitch_Offset_BS}
   AnimGraphNode_StateMachine_3 = Turret Fire SM() {is=title AnimGraphNode_StateMachine}
   n2 = Layered blend per bone(
    BasePose=Sequence Player 'MurkyVole_Pack_Turret_Deployed'() {bLoopAnimation=false, sequence=&MurkyVole_Pack_Turret_Deployed}
    BlendPoses_0=@AnimGraphNode_StateMachine_3
   ) {layerSetup [{branchFilters [{boneName=turret_barrel}, {boneName=turret_shell_ejector}]}]}
   n3 = Layered blend per bone(
    BasePose=@n2
    BlendPoses_0=Blendspace Player 'MurkyVole_Turret_Rail_BS'(X=0) {blendSpace=&MurkyVole_Turret_Rail_BS}
   ) {layerSetup [{branchFilters [{boneName=turret_gun}, {blendDepth=-1, boneName=turret_barrel}, {blendDepth=-1, boneName=turret_shell_ejector}]}]}
   n4 = "Transform (Modify) Bone - Bone: turret_rail"(
    ComponentPose=Local To Component(LocalPose=Apply Additive(Base=@n3, Additive=@n1) {alphaBoolBlend {blendOption=Linear}})
    Alpha=1
    Translation="0,0,0"
    Rotation=Get AimRotationCurrent() -> (AimRotationCurrent:Rotator)
    Scale="1,1,1"
   ) {rotationMode=BMM_Replace, rotationSpace=BCS_WorldSpace, alphaBoolBlend {blendOption=Linear}, boneToModify {boneName=turret_rail}}
   n5 = "Transform (Modify) Bone - Bone: turret_gun"(ComponentPose=@n4, Alpha=1, Translation="0,0,0", Rotation=Get AimRotationCurrent(), Scale="1,1,1") {rotationMode=BMM_Replace, rotationSpace=BCS_WorldSpace, alphaBoolBlend {blendOption=Linear}, boneToModify {boneName=turret_gun}}
   Output Animation Pose(Result=Component To Local(ComponentPose=@n5))
  ]
 }
 Turret Fire SM {
  entry=Wait
  path=AnimGraph.AnimGraphNode_StateMachine_1.Turret SM.AnimStateNode_3.Attacking.AnimGraphNode_StateMachine_3.Turret Fire SM
  ids=id
  nodes [
   To Refire() {is=title AnimStateAliasNode, also {aliases [Fire,Wait]}}
   To Fire() {is=title AnimStateAliasNode, also {aliases [ReFire,Wait]}}
   To Wait() {is=title AnimStateAliasNode, also {aliases [Fire,ReFire]}}
   AnimStateNode_0 = Fire() {is=title AnimStateNode}
   AnimStateNode_1 = ReFire() {is=title AnimStateNode}
   AnimStateNode_2 = Wait() {is=title AnimStateNode}
   AnimStateTransitionNode_0 = AnimStateTransitionNode() {automaticRuleTriggerTime=0, blendMode=Linear, crossfadeDuration=0}
   AnimStateTransitionNode_1 = AnimStateTransitionNode() {automaticRuleTriggerTime=0, blendMode=Linear, crossfadeDuration=0}
   AnimStateTransitionNode_2 = AnimStateTransitionNode() {automaticRuleTriggerTime=0, bAutomaticRuleBasedOnSequencePlayerInState=true, blendMode=Linear, crossfadeDuration=0}
  ]
  links [
   {
    dst=Fire
    rule=AnimGraph.AnimGraphNode_StateMachine_1.Turret SM.AnimStateNode_3.Attacking.AnimGraphNode_StateMachine_3.Turret Fire SM.AnimStateTransitionNode_1.Transition
    src=To Fire
    via=AnimStateTransitionNode_1
   }
   {
    dst=ReFire
    rule=AnimGraph.AnimGraphNode_StateMachine_1.Turret SM.AnimStateNode_3.Attacking.AnimGraphNode_StateMachine_3.Turret Fire SM.AnimStateTransitionNode_0.Transition
    src=To Refire
    via=AnimStateTransitionNode_0
   }
   {
    dst=Wait
    rule=AnimGraph.AnimGraphNode_StateMachine_1.Turret SM.AnimStateNode_3.Attacking.AnimGraphNode_StateMachine_3.Turret Fire SM.AnimStateTransitionNode_2.Transition
    src=To Wait
    via=AnimStateTransitionNode_2
   }
  ]
 }
 Fire {
  path=AnimGraph.AnimGraphNode_StateMachine_1.Turret SM.AnimStateNode_3.Attacking.AnimGraphNode_StateMachine_3.Turret Fire SM.AnimStateNode_0.Fire
  nodes [Output Animation Pose(Result=Sequence Player 'MurkyVole_Turret_Fire_Single'() {bLoopAnimation=false, sequence=/MurkyVoleGameplay/Assets/Backpack/Animations/Turret/MurkyVole_Turret_Fire_Single})]
 }
 ReFire {
  path=AnimGraph.AnimGraphNode_StateMachine_1.Turret SM.AnimStateNode_3.Attacking.AnimGraphNode_StateMachine_3.Turret Fire SM.AnimStateNode_1.ReFire
  nodes [Output Animation Pose(Result=Sequence Player 'MurkyVole_Turret_Fire_Single'() {bLoopAnimation=false, sequence=&MurkyVole_Turret_Fire_Single})]
 }
 Wait {
  path=AnimGraph.AnimGraphNode_StateMachine_1.Turret SM.AnimStateNode_3.Attacking.AnimGraphNode_StateMachine_3.Turret Fire SM.AnimStateNode_2.Wait
  nodes [Output Animation Pose(Result=Sequence Player 'MurkyVole_Pack_Turret_Deployed'() {bLoopAnimation=false, sequence=&MurkyVole_Pack_Turret_Deployed})]
 }
 Transition {
  path=AnimGraph.AnimGraphNode_StateMachine_1.Turret SM.AnimStateNode_3.Attacking.AnimGraphNode_StateMachine_3.Turret Fire SM.AnimStateTransitionNode_0.Transition
  nodes [Result(bCanEnterTransition=Get bIsTurretReFire() -> (bIsTurretReFire:bool))]
 }
 Transition {
  path=AnimGraph.AnimGraphNode_StateMachine_1.Turret SM.AnimStateNode_3.Attacking.AnimGraphNode_StateMachine_3.Turret Fire SM.AnimStateTransitionNode_1.Transition
  nodes [Result(bCanEnterTransition=Get bIsTurretFire() -> (bIsTurretFire:bool))]
 }
 Transition {
  path=AnimGraph.AnimGraphNode_StateMachine_1.Turret SM.AnimStateNode_3.Attacking.AnimGraphNode_StateMachine_3.Turret Fire SM.AnimStateTransitionNode_2.Transition
  nodes [Result()]
 }
 Equip {
  path=AnimGraph.AnimGraphNode_StateMachine_1.Turret SM.AnimStateNode_4.Equip
  nodes [Output Animation Pose(Result=Sequence Player 'MurkyVole_Pack_Turret_Equip'() {bLoopAnimation=false, sequence=/MurkyVoleGameplay/Assets/Backpack/Animations/Turret/MurkyVole_Pack_Turret_Equip})]
 }
 UnEquip {
  path=AnimGraph.AnimGraphNode_StateMachine_1.Turret SM.AnimStateNode_5.UnEquip
  nodes [Output Animation Pose(Result=Sequence Player 'MurkyVole_Pack_Turret_UnEquip'() {bLoopAnimation=false, sequence=/MurkyVoleGameplay/Assets/Backpack/Animations/Turret/MurkyVole_Pack_Turret_UnEquip})]
 }
 Transition {
  path=AnimGraph.AnimGraphNode_StateMachine_1.Turret SM.AnimStateTransitionNode_0.Transition
  nodes [Result(bCanEnterTransition=Get bIsAutoTargetOff() -> (bIsAutoTargetOff:bool))]
 }
 Transition {
  path=AnimGraph.AnimGraphNode_StateMachine_1.Turret SM.AnimStateTransitionNode_1.Transition
  nodes [Result(bCanEnterTransition=Get bIsWindUp() -> (bIsWindUp:bool))]
 }
 Transition {
  path=AnimGraph.AnimGraphNode_StateMachine_1.Turret SM.AnimStateTransitionNode_10.Transition
  nodes [Result(bCanEnterTransition=NOT Boolean(A=Get bIsAutoTargetOff()))]
 }
 Transition {
  path=AnimGraph.AnimGraphNode_StateMachine_1.Turret SM.AnimStateTransitionNode_11.Transition
  nodes [Result(bCanEnterTransition=Get bIsAutoTargetOff())]
 }
 Transition {path=AnimGraph.AnimGraphNode_StateMachine_1.Turret SM.AnimStateTransitionNode_13.Transition, nodes [Result()]}
 Transition {
  path=AnimGraph.AnimGraphNode_StateMachine_1.Turret SM.AnimStateTransitionNode_14.Transition
  nodes [Result(bCanEnterTransition=Get bHasTurret() -> (bHasTurret:bool))]
 }
 Transition {
  path=AnimGraph.AnimGraphNode_StateMachine_1.Turret SM.AnimStateTransitionNode_2.Transition
  nodes [Result(bCanEnterTransition=NOT Boolean(A=Get bHasTurret()))]
 }
 Transition {path=AnimGraph.AnimGraphNode_StateMachine_1.Turret SM.AnimStateTransitionNode_4.Transition, nodes [Result()]}
 Transition {
  path=AnimGraph.AnimGraphNode_StateMachine_1.Turret SM.AnimStateTransitionNode_6.Transition
  nodes [Result(bCanEnterTransition=Get bIsWindUp())]
 }
 Transition {
  path=AnimGraph.AnimGraphNode_StateMachine_1.Turret SM.AnimStateTransitionNode_7.Transition
  nodes [Result(bCanEnterTransition=Get bIsWindDown() -> (bIsWindDown:bool))]
 }
 RocketPod SM {
  entry=Undeployed
  path=AnimGraph.AnimGraphNode_StateMachine_2.RocketPod SM
  ids=id
  nodes [
   To UnEquip Inactive() {is=title AnimStateAliasNode, also {aliases [Inactive]}}
   To AshtonShield Blocking Inactive() {is=title AnimStateAliasNode, also {aliases [Active,Equip,Inactive]}}
   To UnEquip Active() {is=title AnimStateAliasNode, also {aliases [Active,Equip]}}
   AnimStateNode_0 = Undeployed()
   AnimStateNode_1 = Active() {is=title AnimStateNode}
   AnimStateNode_12 = Equip()
   AnimStateNode_2 = Inactive() {is=title AnimStateNode}
   AnimStateNode_3 = Fake Equip to Inacitive() {is=title AnimStateNode}
   AnimStateNode_4 = Alt UnEquip Inactive() {is=title AnimStateNode}
   AnimStateNode_5 = AshtonShield Blocking() {is=title AnimStateNode}
   AnimStateNode_6 = UnEquip Inactive() {is=title AnimStateNode}
   AnimStateNode_7 = Active to Inactive() {is=title AnimStateNode}
   AnimStateNode_8 = Inactive to Active() {is=title AnimStateNode}
   AnimStateNode_9 = UnEquip Active() {is=title AnimStateNode}
   AnimStateTransitionNode_0 = AnimStateTransitionNode() {crossfadeDuration=0}
   AnimStateTransitionNode_1 = AnimStateTransitionNode() {bAutomaticRuleBasedOnSequencePlayerInState=true}
   AnimStateTransitionNode_11 = AnimStateTransitionNode()
   AnimStateTransitionNode_12 = AnimStateTransitionNode() {blendMode=ExpIn, crossfadeDuration=0.7}
   AnimStateTransitionNode_13 = AnimStateTransitionNode()
   AnimStateTransitionNode_14 = AnimStateTransitionNode() {bAutomaticRuleBasedOnSequencePlayerInState=true}
   AnimStateTransitionNode_19 = AnimStateTransitionNode() {crossfadeDuration=0}
   AnimStateTransitionNode_2 = AnimStateTransitionNode() {crossfadeDuration=0.05}
   AnimStateTransitionNode_22 = AnimStateTransitionNode()
   AnimStateTransitionNode_3 = AnimStateTransitionNode() {bAutomaticRuleBasedOnSequencePlayerInState=true, crossfadeDuration=0}
   AnimStateTransitionNode_4 = AnimStateTransitionNode() {bAutomaticRuleBasedOnSequencePlayerInState=true, crossfadeDuration=0}
   AnimStateTransitionNode_6 = AnimStateTransitionNode() {crossfadeDuration=0}
   AnimStateTransitionNode_7 = AnimStateTransitionNode() {automaticRuleTriggerTime=0, bAutomaticRuleBasedOnSequencePlayerInState=true}
   AnimStateTransitionNode_8 = AnimStateTransitionNode()
   AnimStateTransitionNode_9 = AnimStateTransitionNode() {bAutomaticRuleBasedOnSequencePlayerInState=true}
  ]
  links [
   {
    dst=Active to Inactive
    rule=AnimGraph.AnimGraphNode_StateMachine_2.RocketPod SM.AnimStateTransitionNode_2.Transition
    src=Active
    via=AnimStateTransitionNode_2
   }
   {
    dst=Inactive
    rule=AnimGraph.AnimGraphNode_StateMachine_2.RocketPod SM.AnimStateTransitionNode_4.Transition
    src=Active to Inactive
    via=AnimStateTransitionNode_4
   }
   {
    dst=AshtonShield Blocking
    rule=AnimGraph.AnimGraphNode_StateMachine_2.RocketPod SM.AnimStateTransitionNode_3.Transition
    src=Alt UnEquip Inactive
    via=AnimStateTransitionNode_3
   }
   {
    dst=Fake Equip to Inacitive
    rule=AnimGraph.AnimGraphNode_StateMachine_2.RocketPod SM.AnimStateTransitionNode_19.Transition
    src=AshtonShield Blocking
    via=AnimStateTransitionNode_19
   }
   {
    dst=Active
    rule=AnimGraph.AnimGraphNode_StateMachine_2.RocketPod SM.AnimStateTransitionNode_1.Transition
    src=Equip
    via=AnimStateTransitionNode_1
   }
   {
    dst=Inactive
    rule=AnimGraph.AnimGraphNode_StateMachine_2.RocketPod SM.AnimStateTransitionNode_12.Transition
    src=Fake Equip to Inacitive
    via=AnimStateTransitionNode_12
   }
   {
    dst=Inactive to Active
    rule=AnimGraph.AnimGraphNode_StateMachine_2.RocketPod SM.AnimStateTransitionNode_22.Transition
    src=Fake Equip to Inacitive
    via=AnimStateTransitionNode_22
   }
   {
    dst=Inactive to Active
    rule=AnimGraph.AnimGraphNode_StateMachine_2.RocketPod SM.AnimStateTransitionNode_6.Transition
    src=Inactive
    via=AnimStateTransitionNode_6
   }
   {
    dst=Active
    rule=AnimGraph.AnimGraphNode_StateMachine_2.RocketPod SM.AnimStateTransitionNode_9.Transition
    src=Inactive to Active
    via=AnimStateTransitionNode_9
   }
   {
    dst=Alt UnEquip Inactive
    rule=AnimGraph.AnimGraphNode_StateMachine_2.RocketPod SM.AnimStateTransitionNode_11.AnimationTransitionGraph_0
    src=To AshtonShield Blocking Inactive
    via=AnimStateTransitionNode_11
   }
   {
    dst=UnEquip Active
    rule=AnimGraph.AnimGraphNode_StateMachine_2.RocketPod SM.AnimStateTransitionNode_8.Transition
    src=To UnEquip Active
    via=AnimStateTransitionNode_8
   }
   {
    dst=UnEquip Inactive
    rule=AnimGraph.AnimGraphNode_StateMachine_2.RocketPod SM.AnimStateTransitionNode_13.Transition
    src=To UnEquip Inactive
    via=AnimStateTransitionNode_13
   }
   {
    dst=Undeployed
    rule=AnimGraph.AnimGraphNode_StateMachine_2.RocketPod SM.AnimStateTransitionNode_7.Transition
    src=UnEquip Active
    via=AnimStateTransitionNode_7
   }
   {
    dst=Undeployed
    rule=AnimGraph.AnimGraphNode_StateMachine_2.RocketPod SM.AnimStateTransitionNode_14.Transition
    src=UnEquip Inactive
    via=AnimStateTransitionNode_14
   }
   {
    dst=Equip
    rule=AnimGraph.AnimGraphNode_StateMachine_2.RocketPod SM.AnimStateTransitionNode_0.Transition
    src=Undeployed
    via=AnimStateTransitionNode_0
   }
  ]
 }
 Undeployed {
  path=AnimGraph.AnimGraphNode_StateMachine_2.RocketPod SM.AnimStateNode_0.Undeployed
  nodes [Output Animation Pose(Result=Sequence Player 'MurkyVole_Pack_Default'() {bLoopAnimation=false, sequence=&MurkyVole_Pack_Default})]
 }
 Active {
  path=AnimGraph.AnimGraphNode_StateMachine_2.RocketPod SM.AnimStateNode_1.Active
  nodes [
   n1 = Blendspace Player 'MurkyVole_Turret_Pitch_Offset_BS'(X=0) {blendSpace=&MurkyVole_Turret_Pitch_Offset_BS}
   n2 = Sequence Player 'MurkyVole_Pack_Rockets_Deployed'() {bLoopAnimation=false, sequence=/MurkyVoleGameplay/Assets/Backpack/Animations/MurkyVole_Pack_Rockets_Deployed}
   n3 = Get AdjustedRocketPodRot() -> (AdjustedRocketPodRot:Rotator)
   n4 = "Transform (Modify) Bone - Bone: rocket_servo_02_l"(
    ComponentPose=Local To Component(LocalPose=Apply Additive(Base=@n2, Additive=@n1) {alphaBoolBlend {blendOption=Linear}})
    Alpha=0.7
    Translation="0,0,0"
    Rotation=@n3
    Scale="1,1,1"
   ) {rotationMode=BMM_Replace, rotationSpace=BCS_WorldSpace, alphaBoolBlend {blendOption=Linear}, boneToModify {boneName=rocket_servo_02_l}}
   n5 = "Transform (Modify) Bone - Bone: rocket_servo_02_r"(ComponentPose=@n4, Alpha=0.7, Translation="0,0,0", Rotation=@n3, Scale="1,1,1") {
    rotationMode=BMM_Replace
    rotationSpace=BCS_WorldSpace
    translationSpace=BCS_WorldSpace
    alphaBoolBlend {blendOption=Linear}
    boneToModify {boneName=rocket_servo_02_r}
   }
   Output Animation Pose(Result=Component To Local(ComponentPose=@n5))
  ]
 }
 Equip {
  path=AnimGraph.AnimGraphNode_StateMachine_2.RocketPod SM.AnimStateNode_12.Equip
  nodes [Output Animation Pose(Result=Sequence Player 'MurkyVole_Pack_Rockets_Equip'() {bLoopAnimation=false, sequence=/MurkyVoleGameplay/Assets/Backpack/Animations/Rockets/MurkyVole_Pack_Rockets_Equip})]
 }
 Inactive {
  path=AnimGraph.AnimGraphNode_StateMachine_2.RocketPod SM.AnimStateNode_2.Inactive
  nodes [
   n1 = Layered blend per bone(
    BasePose=Blend Poses by bool(
     BlendPose_0=Sequence Player 'MurkyVole_Pack_Crouch_Moving_Rockets_Inactive'() {bLoopAnimation=false, sequence=&MurkyVole_Pack_Crouch_Moving_Rockets_Inactive}
     BlendPose_1=Sequence Player 'MurkyVole_Pack_Rockets_Inactive'() {bLoopAnimation=false, sequence=/MurkyVoleGameplay/Assets/Backpack/Animations/Rockets/MurkyVole_Pack_Rockets_Inactive}
    ) {blendType=Linear}
    BlendPoses_0=Sequence Player 'MurkyVole_Pack_Rockets_Inactive'() {bLoopAnimation=false, sequence=&MurkyVole_Pack_Rockets_Inactive}
   ) {layerSetup [{branchFilters [{boneName=rocket_servo_02_l}, {boneName=rocket_servo_02_r}]}]}
   Output Animation Pose(Result=@n1)
  ]
 }
 Fake Equip to Inacitive {
  path=AnimGraph.AnimGraphNode_StateMachine_2.RocketPod SM.AnimStateNode_3.Fake Equip to Inacitive
  nodes [
   Sequence Player 'MurkyVole_Pack_Rockets_Equip'() {bLoopAnimation=false, sequence=&MurkyVole_Pack_Rockets_Equip}
   Output Animation Pose(Result=Sequence Player 'MurkyVole_Pack_Rockets_Equip_To_Inactive'() {bLoopAnimation=false, sequence=/MurkyVoleGameplay/Assets/Backpack/Animations/Rockets/MurkyVole_Pack_Rockets_Equip_To_Inactive})
  ]
 }
 Alt UnEquip Inactive {
  path=AnimGraph.AnimGraphNode_StateMachine_2.RocketPod SM.AnimStateNode_4.Alt UnEquip Inactive
  nodes [Output Animation Pose(Result=Sequence Player 'MurkyVole_Pack_Rockets_Inactive_To_UnEquipped'() {bLoopAnimation=false, sequence=/MurkyVoleGameplay/Assets/Backpack/Animations/Rockets/MurkyVole_Pack_Rockets_Inactive_To_UnEquipped})]
 }
 AshtonShield Blocking {
  path=AnimGraph.AnimGraphNode_StateMachine_2.RocketPod SM.AnimStateNode_5.AshtonShield Blocking
  nodes [Output Animation Pose(Result=Sequence Player 'MurkyVole_Pack_Default'() {bLoopAnimation=false, sequence=&MurkyVole_Pack_Default})]
 }
 UnEquip Inactive {
  path=AnimGraph.AnimGraphNode_StateMachine_2.RocketPod SM.AnimStateNode_6.UnEquip Inactive
  nodes [Output Animation Pose(Result=Sequence Player 'MurkyVole_Pack_Rockets_Inactive_To_UnEquipped'() {bLoopAnimation=false, sequence=&MurkyVole_Pack_Rockets_Inactive_To_UnEquipped})]
 }
 Active to Inactive {
  path=AnimGraph.AnimGraphNode_StateMachine_2.RocketPod SM.AnimStateNode_7.Active to Inactive
  nodes [Output Animation Pose(Result=Sequence Player 'MurkyVole_Pack_Rockets_Deployed_To_Inactive'() {bLoopAnimation=false, sequence=/MurkyVoleGameplay/Assets/Backpack/Animations/Rockets/MurkyVole_Pack_Rockets_Deployed_To_Inactive})]
 }
 Inactive to Active {
  path=AnimGraph.AnimGraphNode_StateMachine_2.RocketPod SM.AnimStateNode_8.Inactive to Active
  nodes [Output Animation Pose(Result=Sequence Player 'MurkyVole_Pack_Rockets_Inactive_To_Deployed'() {bLoopAnimation=false, sequence=/MurkyVoleGameplay/Assets/Backpack/Animations/Rockets/MurkyVole_Pack_Rockets_Inactive_To_Deployed})]
 }
 UnEquip Active {
  path=AnimGraph.AnimGraphNode_StateMachine_2.RocketPod SM.AnimStateNode_9.UnEquip Active
  nodes [Output Animation Pose(Result=Sequence Player 'MurkyVole_Pack_Rockets_UnEquip'() {bLoopAnimation=false, sequence=/MurkyVoleGameplay/Assets/Backpack/Animations/Rockets/MurkyVole_Pack_Rockets_UnEquip})]
 }
 Transition {
  path=AnimGraph.AnimGraphNode_StateMachine_2.RocketPod SM.AnimStateTransitionNode_0.Transition
  nodes [Result(bCanEnterTransition=Get bHasMissileSwarm() -> (bHasMissileSwarm:bool))]
 }
 Transition {path=AnimGraph.AnimGraphNode_StateMachine_2.RocketPod SM.AnimStateTransitionNode_1.Transition, nodes [Result()]}
 AnimationTransitionGraph_0 {
  path=AnimGraph.AnimGraphNode_StateMachine_2.RocketPod SM.AnimStateTransitionNode_11.AnimationTransitionGraph_0
  nodes [Result(bCanEnterTransition=Get bIsAshtonShieldBlocking() -> (bIsAshtonShieldBlocking:bool))]
 }
 Transition {
  path=AnimGraph.AnimGraphNode_StateMachine_2.RocketPod SM.AnimStateTransitionNode_12.Transition
  nodes [Result(bCanEnterTransition=NOT Boolean(A=Get bIsActiveMissileSwarm() -> (bIsActiveMissileSwarm:bool)))]
 }
 Transition {
  path=AnimGraph.AnimGraphNode_StateMachine_2.RocketPod SM.AnimStateTransitionNode_13.Transition
  nodes [Result(bCanEnterTransition=NOT Boolean(A=Get bHasMissileSwarm()))]
 }
 Transition {path=AnimGraph.AnimGraphNode_StateMachine_2.RocketPod SM.AnimStateTransitionNode_14.Transition, nodes [Result()]}
 Transition {
  path=AnimGraph.AnimGraphNode_StateMachine_2.RocketPod SM.AnimStateTransitionNode_19.Transition
  nodes [Result(bCanEnterTransition=NOT Boolean(A=Get bIsAshtonShieldBlocking()))]
 }
 Transition {
  path=AnimGraph.AnimGraphNode_StateMachine_2.RocketPod SM.AnimStateTransitionNode_2.Transition
  nodes [Result(bCanEnterTransition=NOT Boolean(A=Get bIsActiveMissileSwarm()))]
 }
 Transition {
  path=AnimGraph.AnimGraphNode_StateMachine_2.RocketPod SM.AnimStateTransitionNode_22.Transition
  nodes [Result(bCanEnterTransition=Get bIsActiveMissileSwarm())]
 }
 Transition {path=AnimGraph.AnimGraphNode_StateMachine_2.RocketPod SM.AnimStateTransitionNode_3.Transition, nodes [Result()]}
 Transition {path=AnimGraph.AnimGraphNode_StateMachine_2.RocketPod SM.AnimStateTransitionNode_4.Transition, nodes [Result()]}
 Transition {
  path=AnimGraph.AnimGraphNode_StateMachine_2.RocketPod SM.AnimStateTransitionNode_6.Transition
  nodes [Result(bCanEnterTransition=Get bIsActiveMissileSwarm())]
 }
 Transition {path=AnimGraph.AnimGraphNode_StateMachine_2.RocketPod SM.AnimStateTransitionNode_7.Transition, nodes [Result()]}
 Transition {
  path=AnimGraph.AnimGraphNode_StateMachine_2.RocketPod SM.AnimStateTransitionNode_8.Transition
  nodes [Result(bCanEnterTransition=NOT Boolean(A=Get bHasMissileSwarm()))]
 }
 Transition {path=AnimGraph.AnimGraphNode_StateMachine_2.RocketPod SM.AnimStateTransitionNode_9.Transition, nodes [Result()]}
 EventGraph {links [], nodes []}
]
variables []
